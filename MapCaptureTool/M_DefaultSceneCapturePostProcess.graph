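# Default post-process material for scene captures in the map capture tool.
# It samples the PostProcessInput0 scene texture and masks its color down to the red channel alone.
# The graph applies no other math, parameters or remapping to that channel.
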Material /MapCaptureTool/M_DefaultSceneCapturePostProcess
graph {
 properties {materialDomain=MD_PostProcess, usedShadingModels=Default Lit}
 nodes [ComponentMask(None=SceneTexture() {sceneTextureId=PPI_PostProcessInput0}.Color) {r=true}]
}
