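# Post-process material for a map capture tool that composites one captured TextureSlice into its grid cell of CombinedTexture.
# Custom HLSL derives the row-major cell from Index, TextureSize and SliceSize, then remaps screen UVs into slice-local space.
# A second custom node masks that cell, selecting slice color inside and existing color outside.
# The mask is added to existing alpha and saturated, marking filled cells opaque, with alpha output enabled.
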Material /MapCaptureTool/M_StitchMapCapture
graph {
 properties {blendableOutputAlpha=true, floatPrecisionMode=MFPM_Full, materialDomain=MD_PostProcess, usedShadingModels=Default Lit}
 nodes [
  n1 = TextureSampleParameter2D() {parameterName=CombinedTexture}
  MaterialExpressionVectorParameter_2 = VectorParameter() {parameterName=TextureSize, defaultValue=<rgba 4096 4096 0 1>}
  MaterialExpressionVectorParameter_3 = VectorParameter() {parameterName=SliceSize, defaultValue=<rgba 1024 1024 0 1>}
  MaterialExpressionScalarParameter_2 = ScalarParameter() {parameterName=Index}
  MaterialExpressionTextureCoordinate_1 = TextureCoordinate()
  MaterialExpressionCustom_1 = Custom!(
   TextureSize=@MaterialExpressionVectorParameter_2.RGB
   SliceSize=@MaterialExpressionVectorParameter_3.RGB
   SliceIndex=@MaterialExpressionScalarParameter_2
   TexCoord=@MaterialExpressionTextureCoordinate_1
  ) {
   code="float SlicesPerRow = floor(TextureSize.x / SliceSize.x);\r\nfloat SlicesPerColumn = floor(TextureSize.y / SliceSize.y);\r\n\r\nint ColumnIndex = SliceIndex % SlicesPerRow;\r\nint RowIndex = floor(SliceIndex / SlicesPerRow);\r\n\r\nfloat2 Scale = float2(1.0 / SlicesPerRow, 1.0 / SlicesPerColumn);\r\nfloat2 Offset = float2(ColumnIndex, RowIndex);\r\n\r\nreturn TexCoord / Scale - Offset;"
   outputType=CMOT_Float2
   showCode=true
   inputs [
    {
     inputName=TextureSize
     input {expression=&self:MaterialExpressionVectorParameter_2, mask=1, maskB=1, maskG=1, maskR=1}
    }
    {
     inputName=SliceSize
     input {expression=&self:MaterialExpressionVectorParameter_3, mask=1, maskB=1, maskG=1, maskR=1}
    }
    {inputName=SliceIndex, input {expression=&self:MaterialExpressionScalarParameter_2}}
    {inputName=TexCoord, input {expression=&self:MaterialExpressionTextureCoordinate_1}}
   ]
  }
  n2 = Custom!(UVs=@MaterialExpressionCustom_1) {
   code="if (UVs.x > 1.0 || UVs.x < 0.0 || UVs.y > 1.0 || UVs.y < 0.0)\r\n    return 0;\r\nreturn 1;"
   outputType=CMOT_Float1
   showCode=true
   inputs [{inputName=UVs, input {expression=&self:MaterialExpressionCustom_1}}]
  }
  LinearInterpolate(A=@n1.RGB, B=TextureSampleParameter2D(UVs=Saturate(None=@MaterialExpressionCustom_1)) {parameterName=TextureSlice}.RGB, Alpha=@n2)
  Saturate(None=Add(A=@n1.A, B=@n2))
 ]
}
