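# Editor utility task that captures a World Partition map as a Slices-by-Slices grid, logging timestamps at start and end.
# After BeginCapture and PrewarmPasses, each slice loads, waits for loading and LODs, hides actors, is processed, unhides and unloads.
# A manual-iteration loop macro advances only after these latent steps and the post-unload wait finish.
# Cancellation stops iteration after the current slice, without calling EndCapture.
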EditorUtilityBlueprint /MapCaptureTool/MapCaptureTask
graphs [
 EventGraph {nodes [
   Event BeginExecution()
   > PrintTime()
   > Set CancelQueued(CancelQueued:bool=false) -> (Output_Get:bool)
   > BeginCapture(self:Map Capture Tool Utility Task Object Reference)
   > DelayUntilNextFrame!(execute, WorldContextObject:Object Reference) -> (then)
   > PrewarmPasses(self:Map Capture Tool Utility Task Object Reference)
   > DelayUntilNextFrame()
   n1 = Get Slices() -> (Slices:int=0)
   > n2 = For Loop Manual Iteration!(execute:Exec, Iterate:Exec=@n5.else, FirstIndex:int, LastIndex:int=int * int(A=@n1, B=@n1)) -> (LoopBody:Exec, Index:int, Completed:Exec) {is=title K2Node_MacroInstance}
   EndCapture(execute=@n2.Completed, self:Map Capture Tool Utility Task Object Reference)
   > PrintTime()
   LoadWorldPartitionSlice!(execute=@n2.LoopBody, self:Map Capture Tool Utility Task Object Reference, SliceIndex:int=@n2, UnloadDelay:float=8) -> (then)
   > n3 = WaitForBackgroundLoadingAndLODs!(execute:Exec, Timeout:float=120, WorldContextObject:Object Reference=Self-Reference()) -> (then:Exec, Complete:Exec, bSuccess:bool=~) {is=title K2Node_AsyncAction}
   Delay!(execute=@n3.Complete, WorldContextObject:Object Reference, Duration=3) -> (then)
   > DelayUntilNextFrame()
   > HideActors(self:Map Capture Tool Utility Task Object Reference)
   > ProcessSlice!(execute, self:Map Capture Tool Utility Task Object Reference, SliceIndex:int=@n2) -> (then)
   > DelayUntilNextFrame()
   > UnhideActors(self:Map Capture Tool Utility Task Object Reference)
   > DelayUntilNextFrame()
   > UnloadWorldPartitionSlice!(execute, self:Map Capture Tool Utility Task Object Reference, SliceIndex:int=@n2) -> (then)
   > n4 = WaitForBackgroundLoadingAndLODs(Timeout=120, WorldContextObject=Self-Reference()) -> (bSuccess=~)
   n5 = Branch(execute=@n4.Complete, Condition=Get CancelQueued() -> (CancelQueued:bool))
   Event CancelRequested() {is=title K2Node_Event}
   > Set CancelQueued(CancelQueued=true)
  ]}
 ForLoop ManualIteration {nodes [
   n1 = Inputs!() -> (execute, Iterate:Exec, FirstIndex:int, LastIndex:int)
   n2 = Local Integer!() -> (Variable:int) {is=title K2Node_TemporaryVariable}
   n3 = Assign!(execute:Exec=@n1.Iterate, Variable:int=@n2, Value:int=int + int!(A:int=@n2, B:int=1) -> (ReturnValue:int=0) {is=title K2Node_CallFunction}) -> (then:Exec) {is=title K2Node_AssignmentStatement}
   n4 = Branch(execute=[@n5.then, @n3.then], Condition="integer < integer"(A=@n2, B=@n1.LastIndex))
   > Outputs!(LoopBody:Exec, Index:int=@n2, Completed:Exec=@n4.else)
   n5 = Assign(execute=@n1.execute, Variable=@n2, Value=@n1)
  ]}
 PrintTime {nodes [
   PrintTime() {is=title K2Node_FunctionEntry}
   > n1 = GetPlatformTime_Seconds!(execute) -> (then, ReturnValue:double)
   > PrintString(InString=Append(A=MakeLiteralString(Value=MapCapture :: Timestamp), B="To String (Float)"(InDouble=@n1)))
  ]}
]
variables [CancelQueued {type=boolean}]
