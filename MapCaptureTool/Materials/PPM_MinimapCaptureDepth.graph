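# Post-process material for minimap capture that, by default, turns scene depth into a grayscale height image through divide, power and saturate.
# Nested static switches can instead show custom stencil, a shading-model tree mask, metallic, or a color-graded image.
# The color branch tints water found by shading-model ID threshold and applies an Apollo terrain mask vignette, outputting black on forward and mobile paths.
# Several custom-depth and mask node chains are left unconnected.
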Material /MapCaptureTool/Materials/PPM_MinimapCaptureDepth
graph {
 properties {blendableOutputAlpha=true, materialDomain=MD_PostProcess, pixelDepthOffsetMode=PDOM_Legacy, usedShadingModels=Default Lit}
 nodes [
  Constant() {r=1}
  n1 = Constant()
  If(
   A=ComponentMask(None=SceneTexture() {sceneTextureId=PPI_ShadingModelID}.Color) {r=true}
   B=ScalarParameter() {defaultValue=10, parameterName=SahdingModelWater}
   "A > B"=@n1
   "A == B"=Constant() {r=1}
   "A < B"=@n1
  )
  n2 = Constant()
  If(
   A=ComponentMask(None=SceneTexture() {sceneTextureId=PPI_ShadingModelID}.Color) {r=true}
   B=ScalarParameter() {defaultValue=10, parameterName=SahdingModelWater}
   "A > B"=@n2
   "A == B"=Constant() {r=1}
   "A < B"=@n2
  )
  n3 = ComponentMask(None=SceneTexture() {sceneTextureId=PPI_CustomDepth}.Color) {r=true}
  Multiply(A=@n3) {constB=1000000}
  n4 = Subtract(
   A=ComponentMask(None=SceneTexture() {sceneTextureId=PPI_ShadingModelID}.Color) {r=true}
   B=ScalarParameter() {defaultValue=7.4323, parameterName=Threshold1}
  )
  OneMinus(None=Saturate(None=Ceil(None=@n4)))
  Power(Base=Divide(A=@n3) {constB=100000}, Exp=ScalarParameter() {defaultValue=5, parameterName=MapHeightPower})
  n5 = Subtract(
   A=ComponentMask(None=SceneTexture() {sceneTextureId=PPI_ShadingModelID}.Color) {r=true}
   B=ScalarParameter() {defaultValue=7.4323, parameterName=Threshold1}
  )
  n6 = OneMinus(None=Saturate(None=Ceil(None=@n5)))
  Reroute(None=@n6)
  Saturate(None=Power(Base=Divide(A=@n3) {constB=1}) {constExponent=1})
  ScalarParameter() {defaultValue=100000, parameterName=MapHeightDivision}
  n7 = TextureSample(UVs=ScreenPosition().ViewportUV) {samplerType=SAMPLERTYPE_Masks, texture=/Game/Athena/Apollo/Maps/UI/Apollo_Terrain_Minimap_PostMask}
  n8 = LinearInterpolate(A=Constant3Vector() {constant=<rgba 1 1 1 0>}, B=VectorParameter() {parameterName=Color2, defaultValue=<rgba 1 1 1 1>}.RGB, Alpha=@n7.R)
  n9 = ComponentMask(None=SceneTexture() {sceneTextureId=PPI_PostProcessInput0}.Color) {b=true, g=true, r=true}
  n10 = ScalarParameter() {defaultValue=1, parameterName=Power}
  n11 = Desaturation(
   None=Multiply(A=Power(Base=@n9, Exp=@n10), B=ScalarParameter() {defaultValue=1, parameterName=Brightness})
   Fraction=ScalarParameter() {defaultValue=0.05, parameterName=Saturation}
  )
  n12 = LinearInterpolate(
   A=LinearInterpolate(
    A=VectorParameter() {parameterName=Color1Water, defaultValue=<rgba 0.0189 0.1369 0.3906 1>}.RGB
    B=@n11
    Alpha=ScalarParameter() {defaultValue=0.85, parameterName=WaterBlend}
   )
   B=@n11
   Alpha=@n6
  )
  n13 = LinearInterpolate(
   A=ComponentMask(None=SceneTexture() {sceneTextureId=PPI_PostProcessInput0}.Color) {b=true, g=true, r=true}
   B=LinearInterpolate(
    A=VectorParameter() {parameterName=Color3WaterVignette, defaultValue=<rgba 0.0008 0.0073 0.02 1>}.RGB
    B=Multiply(A=@n8, B=@n12)
    Alpha=@n7.R
   )
   Alpha=ScalarParameter() {defaultValue=1, parameterName=Enabled}
  )
  n14 = Constant3Vector()
  n15 = ShadingPathSwitch(Default=@n13, Deferred=@n13, Forward=@n14, Mobile=@n14)
  n16 = Constant()
  n17 = If(
   A=ComponentMask(None=SceneTexture() {sceneTextureId=PPI_ShadingModelID}.Color) {r=true}
   B=ScalarParameter() {defaultValue=2, parameterName=SahdingModelTree}
   "A > B"=@n16
   "A == B"=Constant() {r=1}
   "A < B"=@n16
  )
  n18 = Divide(A=ComponentMask(None=SceneTexture() {sceneTextureId=PPI_SceneDepth}.Color) {b=true, g=true, r=true}) {constB=30000}
  n19 = ScalarParameter() {defaultValue=5, parameterName=MapHeightPower}
  n20 = StaticSwitchParameter(True=SceneTexture() {sceneTextureId=PPI_CustomStencil}.Color, False=Saturate(None=Power(Base=@n18, Exp=@n19))) {parameterName=ShowCustomDepth}
  StaticSwitchParameter(
   True=SceneTexture() {sceneTextureId=PPI_Metallic}.Color
   False=StaticSwitchParameter(True=Add(B=@n15), False=StaticSwitchParameter(True=@n17, False=@n20) {parameterName=EnableTreeMask}) {parameterName=EnableWaterMask}
  ) {parameterName=EnableMetallicMask}
 ]
}
